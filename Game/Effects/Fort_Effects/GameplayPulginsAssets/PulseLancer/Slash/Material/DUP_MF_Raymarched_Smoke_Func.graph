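# Material function for lit, raymarched-look flipbook smoke particles, kept with the PulseLancer slash effect assets.
# A custom HLSL loop marches distorted 2D texture UVs along the sprite-rotated light direction, accumulating transmittance with time-of-day density and view-dependent AO.
# Emissive combines that lighting with particle color, backlight and optional spherical glow, simplified at lower quality.
# Opacity and dithered mask outputs use dynamic-parameter erosion and near-camera fade, with depth fade off mobile.
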MaterialFunction /Game/Effects/Fort_Effects/GameplayPulginsAssets/PulseLancer/Slash/Material/DUP_MF_Raymarched_Smoke_Func
graph {
 properties {previewBlendMode=BLEND_Translucent}
 nodes [
  CameraVectorWS()
  CollectionParameter() {collection=&FortniteMaterialParameters, parameterName=SunLightColor}
  Constant3Vector() {constant=<rgba 1 0 0 0>}
  n1 = VectorParameter() {parameterName=Distortl SubUVTiling, defaultValue=<rgba 8 8 0 0>}
  n2 = TextureSampleParameter2D(UVs=Multiply(A=TextureCoordinate(), B=ComponentMask(None=@n1.RGB) {g=true, r=true})) {parameterName=DistortTexture, samplerType=SAMPLERTYPE_Normal, texture=/Game/VisualThreatManager/StormVisuals/LargeClouds/T_StormSwirlNormal}
  n3 = Multiply(A=ComponentMask(None=@n2.RGB) {g=true, r=true}, B=ScalarParameter() {defaultValue=0.005, group=Shape Control, parameterName=UV Distort})
  n4 = Add(A=TextureCoordinate(), B=@n3)
  n5 = NamedRerouteDeclaration(None=@n4) {name=UV Distortion, nodeColor=<rgba 0 1 0.6953 1>}
  MaterialExpressionTextureObjectParameter_0 = TextureObjectParameter(UVs=@n4) {parameterName=TextureObject, samplerType=SAMPLERTYPE_LinearColor, texture=/Game/Effects/Fort_Effects/Textures/Smoke/T_FallingMist_8x8_01}
  n6 = TextureSample(UVs=@n5, Tex=@MaterialExpressionTextureObjectParameter_0)
  FunctionOutput(None=@n6.RGBA) {outputName=SampleResult}
  n7 = ParticleColor()
  n8 = Saturate(None=DotProduct(A=@n6.RGB, B=NamedRerouteUsage()))
  n9 = ScalarParameter() {defaultValue=250, group=Sphereical Glow, parameterName=Glow distort}
  n10 = SphereMask(
   A=Add(A=WorldPosition().XYZ, B=Multiply(A=@n8, B=@n9))
   B=Add(A=ActorPositionWS(), B=VectorParameter() {group=Sphereical Glow, parameterName=Glow Offset}.RGB)
   Radius=ScalarParameter() {defaultValue=750, group=Sphereical Glow, parameterName=Glow Radius}
   Hardness=ScalarParameter() {defaultValue=0.5, group=Sphereical Glow, parameterName=Glow Hardness}
  )
  n11 = VectorParameter() {group=Sphereical Glow, parameterName=Flame Glow, defaultValue=<rgba 2 0.1667 0 1>}
  n12 = Multiply(A=Multiply(A=Multiply(A=@n10, B=@n11.RGB), B=@n8), B=@n8)
  MaterialExpressionScalarParameter_2 = ScalarParameter() {defaultValue=0.15, group=Raymarch | Controls, parameterName=MarchDistance}
  MaterialExpressionScalarParameter_1 = ScalarParameter() {defaultValue=32, group=Raymarch | Controls, parameterName=NumSteps}
  MaterialExpressionScalarParameter_16 = ScalarParameter() {defaultValue=1, group=Raymarch | Controls, parameterName=Transmitted Light}
  MaterialExpressionMultiply_11 = Multiply(
   A=ComponentMask(None=CollectionParameter() {collection=&FortniteMaterialParameters, parameterName=CloudsShadow}) {b=true, g=true, r=true}
   B=ScalarParameter() {defaultValue=1, group=Shape Control, parameterName=CloudShadow Color Scale}
  )
  MaterialExpressionMultiply_4 = Multiply(
   A=ScalarParameter() {defaultValue=1, group=Raymarch | Density, parameterName=Overall Density}
   B=MaterialFunctionCall(
    Value_Morning=ScalarParameter() {defaultValue=1, group=Raymarch | Density, parameterName=Density Scale Morning}
    Value_Day=ScalarParameter() {defaultValue=1.25, group=Raymarch | Density, parameterName=Density Scale Day}
    Value_Evening=ScalarParameter() {defaultValue=0.75, group=Raymarch | Density, parameterName=Density Scale Eveneing}
    Value_Night=ScalarParameter() {defaultValue=1.25, group=Raymarch | Density, parameterName=Density Scale Night}
   ) {materialFunction=&Scalar_Time_of_Day}.Result
  )
  MaterialExpressionAdd_4 = Add(A=@n3, B=TextureCoordinate())
  MaterialExpressionRotator_0 = Rotator(
   Coordinate=Transform(None=AtmosphericLightVector()) {transformSourceType=TRANSFORMSOURCE_World, transformType=TRANSFORM_Tangent}
   Time=ParticleSpriteRotation().Rad
  )
  MaterialExpressionNamedRerouteUsage_4 = NamedRerouteUsage()
  MaterialExpressionScalarParameter_6 = ScalarParameter() {defaultValue=4, group=Raymarch | Controls, parameterName=Contrast}
  n13 = VectorParameter() {parameterName=ChannelSelect, defaultValue=<rgba 0 1 0 0>}
  MaterialExpressionMaterialFunctionCall_0 = MaterialFunctionCall(X=@n13.R, Y=@n13.R, Z=@n13.R, A=@n13.R) {materialFunction=&MakeFloat4}
  n14 = Saturate(None=Multiply(A=@n6.R, B=ScalarParameter() {defaultValue=1, group=Detail, parameterName=Detail_GrayAdjust_Mul}))
  n15 = ScalarParameter() {defaultValue=1, group=Detail, parameterName=Detail_GrayAdjust_Power}
  n16 = AtmosphericLightVector()
  n17 = Clamp(None=DotProduct(
    A=MaterialFunctionCall() {materialFunction=&CameraDirectionVector}.Result
    B=Multiply(A=@n16) {constB=-1}
   ))
  n18 = ScalarParameter() {defaultValue=1, group=AOArea, parameterName=AOArea_GrayAdjust_Mul}
  n19 = Power(Base=Saturate(None=Multiply(A=@n17, B=@n18)), Exp=ScalarParameter() {defaultValue=1, group=AOArea, parameterName=AOArea_GrayAdjust_Power})
  MaterialExpressionMaterialFunctionCall_1 = MaterialFunctionCall(
   Input=Saturate(None=Power(Base=@n14, Exp=@n15))
   Input Low=Constant()
   Input High=Constant() {r=1}
   Target Low=LinearInterpolate(A=ScalarParameter() {defaultValue=1, parameterName=B}, B=ScalarParameter() {parameterName=A}, Alpha=Saturate(None=@n19))
   Target High=Constant() {r=1}
  ) {materialFunction=&RemapValueRange}
  n20 = Custom!(
   Tex=@MaterialExpressionTextureObjectParameter_0
   marchDistance=@MaterialExpressionScalarParameter_2
   numSteps=@MaterialExpressionScalarParameter_1
   TransmittedLight=@MaterialExpressionScalarParameter_16
   LO=@MaterialExpressionMultiply_11
   density=@MaterialExpressionMultiply_4
   UV=@MaterialExpressionAdd_4
   marchDir=@MaterialExpressionRotator_0
   LightColor=@MaterialExpressionNamedRerouteUsage_4
   contrast=@MaterialExpressionScalarParameter_6
   ChannelSelect=@MaterialExpressionMaterialFunctionCall_0.Result
   AOIn=@MaterialExpressionMaterialFunctionCall_1.Result
  ) {
   code="for (float d = 0.0; d < marchDistance; d += marchDistance / numSteps)\r\n{\r\n    UV += (marchDir * d) / numSteps;\r\n    float4 textureSample = Texture2DSampleLevel(Tex, TexSampler, UV.xy, 0) / (numSteps - contrast);\r\n    // Extract luminance channel\r\n    float densityChannel = saturate(dot(textureSample, ChannelSelect));\r\n    TransmittedLight *= saturate(1.0 - densityChannel);\r\n\r\n}\r\nfloat3 LI = LightColor * TransmittedLight.xxx * AOIn;\r\nLO += LI * density;\r\nreturn LO;"
   description=2D RayMarching
   inputs [
    {
     inputName=Tex
     input {expression=&self:MaterialExpressionTextureObjectParameter_0}
    }
    {
     inputName=marchDistance
     input {expression=&self:MaterialExpressionScalarParameter_2}
    }
    {
     inputName=numSteps
     input {expression=&self:MaterialExpressionScalarParameter_1}
    }
    {
     inputName=TransmittedLight
     input {expression=&self:MaterialExpressionScalarParameter_16}
    }
    {
     inputName=LO
     input {expression=&self:MaterialExpressionMultiply_11}
    }
    {
     inputName=density
     input {expression=&self:MaterialExpressionMultiply_4}
    }
    {
     inputName=UV
     input {expression=&self:MaterialExpressionAdd_4}
    }
    {
     inputName=marchDir
     input {expression=&self:MaterialExpressionRotator_0}
    }
    {
     inputName=LightColor
     input {expression=&self:MaterialExpressionNamedRerouteUsage_4}
    }
    {
     inputName=contrast
     input {expression=&self:MaterialExpressionScalarParameter_6}
    }
    {
     inputName=ChannelSelect
     input {expression=&self:MaterialExpressionMaterialFunctionCall_0}
    }
    {
     inputName=AOIn
     input {expression=&self:MaterialExpressionMaterialFunctionCall_1}
    }
   ]
  }
  n21 = ScalarParameter() {defaultValue=1, group=Lighting, parameterName=Backlight Power}
  n22 = Multiply(
   A=Multiply(A=Power(Base=@n20, Exp=@n21), B=ScalarParameter() {defaultValue=1, group=Lighting, parameterName=Backlight Scale})
   B=NamedRerouteUsage()
  )
  n23 = Multiply(
   A=StaticSwitchParameter(True=@n7.RGB, False=VectorParameter() {parameterName=Base Color}.RGB) {defaultValue=true, parameterName=Use Particle Color}
   B=Add(A=StaticSwitchParameter(True=@n12, False=@n8) {group=Sphereical Glow, parameterName=Use Sphereical Glow}, B=Add(A=@n20, B=@n22))
  )
  n24 = LinearInterpolate(A=Multiply(A=@n23, B=ScalarParameter() {defaultValue=0.25, parameterName=darken}), B=@n23, Alpha=NamedRerouteUsage())
  n25 = Multiply(
   A=LinearInterpolate(B=ScalarParameter() {defaultValue=1}, Alpha=NamedRerouteUsage())
   B=Multiply(A=ParticleColor().RGB, B=NamedRerouteUsage())
  )
  FunctionOutput(None=QualitySwitch(Default=Add(A=@n24, B=@n25) {constB=0}, Low=@n23, Medium=@n23)) {outputName=Emissive}
  n26 = Power(Base=@n8, Exp=DynamicParameter() {paramNames [Erode Power, Contrast, Param3, Param4]}.Erode Power)
  n27 = Multiply(A=@n7.A, B=@n26)
  n28 = DepthFade(Opacity=@n27, FadeDistance=ScalarParameter() {defaultValue=100, parameterName=FadeDistance})
  n29 = MaterialFunctionCall(Distance=ScalarParameter() {defaultValue=2500, parameterName=Near Camera Fade}) {materialFunction=&Near_Camera_Fade}
  FunctionOutput(None=Multiply(A=ShadingPathSwitch(Default=@n28, Mobile=@n27), B=@n29.Result)) {outputName=Opacity, sortPriority=10}
  n30 = LinearInterpolate(
   B=MaterialFunctionCall!(Alpha Threshold=ScalarParameter() {defaultValue=0.5, parameterName=Param_1}, Random=ParticleRandom()) {materialFunction=/Game/TimeOfDay/CloudsAndStorm/Materials/MaterialFunctions/DitherTemporalAASpatioTemporalBlue}.Result
   Alpha=@n26
  )
  FunctionOutput(None=Multiply(A=@n29.Result, B=Multiply(A=@n7.A, B=@n30))) {outputName=Opacity Mask, sortPriority=20}
  LocalPosition()
  LocalPosition()
  MaterialFunctionCall() {materialFunction=&DitherTemporalAA}
  Multiply()
  n31 = DotProduct(A=Multiply(A=AtmosphericLightVector()) {constB=-1}, B=CameraVectorWS())
  Multiply(B=LinearInterpolate(A=ScalarParameter() {group=Lighting, parameterName=Backlight scale Low}, Alpha=Saturate(None=@n31)))
  Multiply() {constB=-1}
  Multiply() {constB=-1}
  n32 = TextureCoordinate()
  n33 = ComponentMask(None=VectorParameter() {parameterName=World Normal SubUVTiling, defaultValue=<rgba 8 8 0 0>}.RGB) {g=true, r=true}
  n34 = MaterialFunctionCall(
   Normal=TextureSample(UVs=Multiply(A=@n32, B=@n33)) {samplerType=SAMPLERTYPE_Normal, texture=&T_Sphere_N}.RGB
   Flatness=ScalarParameter() {defaultValue=0.1, parameterName=Flatten Sphere Normal}
  ) {materialFunction=&FlattenNormal}
  n35 = Transform(None=AtmosphericLightVector()) {transformSourceType=TRANSFORMSOURCE_World, transformType=TRANSFORM_Tangent}
  n36 = Power(Base=Saturate(None=DotProduct(A=@n34.Result, B=@n35)), Exp=ScalarParameter() {defaultValue=2, parameterName=World Normal Power})
  n37 = Constant3Vector() {constant=<rgba 0 0 1 0>}
  NamedRerouteDeclaration(None=ShadingPathSwitch(Default=QualitySwitch(Default=@n36, Low=@n37, High=@n36, Medium=@n37, Epic=@n36), Mobile=Constant3Vector() {constant=<rgba 0 0 1 0>})) {name=World Normal, nodeColor=<rgba 1 0.4688 0 1>}
  NamedRerouteDeclaration(None=@n13.RGB) {name=Channel Selector, nodeColor=<rgba 1 0 0.0938 1>}
  n38 = TextureSampleParameter2D(UVs=NamedRerouteUsage()) {parameterName=Additional Normal, samplerType=SAMPLERTYPE_Normal, texture=/Game/Effects/Fort_Effects/Textures/Smoke/T_FallingMist_8x8_01_N}
  n39 = MaterialFunctionCall(Normal=@n38.RGB, Flatness=ScalarParameter() {parameterName=Flatten Normal}) {materialFunction=&FlattenNormal}
  n40 = Transform(None=AtmosphericLightVector()) {transformSourceType=TRANSFORMSOURCE_World, transformType=TRANSFORM_Tangent}
  NamedRerouteDeclaration(None=Saturate(None=DotProduct(A=@n39.Result, B=@n40))) {name=Additional Normal Map, nodeColor=<rgba 0 0.1797 1 1>}
  n41 = VectorParameter() {group=Sunlight- Time of Day, parameterName="Color_Morning (V3)", defaultValue=<rgba 0.6062 0.3965 0.3372 1>}
  n42 = VectorParameter() {group=Sunlight- Time of Day, parameterName="Color_Evening (V3)", defaultValue=<rgba 0.5771 0.344 0.2721 1>}
  n43 = MaterialFunctionCall(
   Color_Morning=@n41.RGB
   Color_Day=VectorParameter() {group=Sunlight- Time of Day, parameterName="Color_Day (V3)", defaultValue=<rgba 0.3479 0.28 0.2385 1>}.RGB
   Color_Evening=@n42.RGB
   Color_Night=VectorParameter() {group=Sunlight- Time of Day, parameterName="Color_Night (V3)", defaultValue=<rgba 0.2498 0.294 0.5146 1>}.RGB
  ) {materialFunction=&Color_Time_of_Day}
  NamedRerouteDeclaration(None=@n43.Result) {name=Time Of Day, nodeColor=<rgba 1 0.4453 0 1>}
  OneMinus()
  n44 = TextureCoordinate()
  n45 = ComponentMask(None=VectorParameter() {group=Sphereical Glow, parameterName=Noise Tiling, defaultValue=<rgba 8 8 0 0>}.RGB) {g=true, r=true}
  OneMinus(None=TextureSample(UVs=Multiply(A=@n44, B=@n45)) {texture=/Game/Effects/FogVolume/noise_flamy_Tile}.G)
  ParticleSubUV!(UVs=@n5, Tex=@MaterialExpressionTextureObjectParameter_0, Apply View MipBias) {samplerType=SAMPLERTYPE_Normal, texture=&T_FallingMist_8x8_01_N}
  Saturate(None=DotProduct(
    A=TextureSample() {samplerType=SAMPLERTYPE_LinearColor, texture=/Game/Effects/Fort_Effects/Gameplay/SmokeDustExplosions/Textures/T_SmokeDeath_8x8_01}.RGB
    B=VectorParameter() {parameterName=Param_1, defaultValue=<rgba 1 0 0 0>}.RGB
   ))
  ScalarParameter() {defaultValue=4, parameterName=Param_2}
  n46 = Constant3Vector() {constant=<rgba 0 0 1 0>}
  ShadingPathSwitch(Default=QualitySwitch(Low=@n46, Medium=@n46), Mobile=Constant3Vector() {constant=<rgba 0 0 1 0>})
  StaticSwitchParameter(True=Multiply(A=ParticleColor().RGB, B=@n43.Result), False=Multiply(A=@n43.Result, B=VectorParameter() {parameterName=Base Color}.RGB)) {defaultValue=true, parameterName=Use Particle Color}
  TextureSampleParameterSubUV!(UVs, Apply View MipBias) {parameterName=asdf, texture=null}
  TextureSample()
  Transform(None=VectorParameter() {defaultValue=<rgba 1 0 0 0>}.RGB) {transformSourceType=TRANSFORMSOURCE_World, transformType=TRANSFORM_Instance}
  VectorParameter() {parameterName=Param_2}
  VectorParameter() {parameterName=Custom Facing Light Dir}
  VectorParameter() {parameterName=Param_4}
  VectorParameter() {parameterName=Param_3, defaultValue=<rgba 0 0 0 50>}
  VertexNormalWS()
  WorldPosition()
 ]
}
