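# Material function for the PulseLancer slash effect that outputs procedural Growth, Dissolve and TipArea masks plus a world position offset.
# Growth and Dissolve sweep along vertex color red using normalized time inputs, padded by inverse blend width and shaped by power.
# TipArea derives a band at the growth front from the unscaled growth distance.
# WPOWS blends tangent and binormal, normalizing and scaling in instance space before returning to world space.
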MaterialFunction /Game/Effects/Fort_Effects/GameplayPulginsAssets/PulseLancer/Slash/ProceduralSlashParameters
graph {nodes [
  FunctionInput() {bUsePreviewValueAsDefault=true, inputName=---------------, inputType=FunctionInput_Scalar, sortPriority=5, previewValue=<10 0 0 1>}
  FunctionInput() {bUsePreviewValueAsDefault=true, inputName=----------------, inputType=FunctionInput_Scalar, sortPriority=10, previewValue=<10 0 0 1>}
  FunctionInput() {bUsePreviewValueAsDefault=true, inputName=-----------------, inputType=FunctionInput_Scalar, sortPriority=21, previewValue=<10 0 0 1>}
  n1 = FunctionInput() {bUsePreviewValueAsDefault=true, inputName=GrowthTime0-1, inputType=FunctionInput_Scalar, previewValue=<1 0 0 1>}
  n2 = FunctionInput() {bUsePreviewValueAsDefault=true, inputName=GrowthBlendWidth_Inverse, inputType=FunctionInput_Scalar, sortPriority=1, previewValue=<10 0 0 1>}
  n3 = Add(B=Divide(B=@n2)) {constA=1}
  n4 = Multiply(A=Subtract(A=VertexColor().R, B=Multiply(A=@n1, B=@n3))) {constB=-1}
  n5 = Multiply(A=@n4, B=@n2)
  n6 = FunctionInput() {bUsePreviewValueAsDefault=true, inputName=GrowthBlendPower, inputType=FunctionInput_Scalar, sortPriority=2, previewValue=<1 0 0 1>}
  FunctionOutput(None=Power(Base=Saturate(None=@n5), Exp=@n6)) {outputName=Growth}
  n7 = FunctionInput() {bUsePreviewValueAsDefault=true, inputName=DissolveTime0-1, inputType=FunctionInput_Scalar, sortPriority=13, previewValue=<1 0 0 1>}
  n8 = Constant()
  n9 = Constant() {r=1}
  n10 = FunctionInput() {
   bUsePreviewValueAsDefault=true
   inputName=DissolveBlendWidth_Inverse
   inputType=FunctionInput_Scalar
   sortPriority=14
   previewValue=<10 0 0 1>
  }
  n11 = Subtract(B=Divide(B=@n10)) {constA=0}
  n12 = Multiply(
   A=Subtract(
    A=VertexColor().R
    B=MaterialFunctionCall(Input=@n7, Input Low=@n8, Input High=@n9, Target Low=@n11, Target High=@n9) {materialFunction=&RemapValueRange}.Result
   )
   B=@n10
  )
  n13 = FunctionInput() {bUsePreviewValueAsDefault=true, inputName=DissolveBlendPower, inputType=FunctionInput_Scalar, sortPriority=15, previewValue=<1 0 0 1>}
  FunctionOutput(None=Power(Base=Saturate(None=@n12), Exp=@n13)) {outputName=Dissolve, sortPriority=10}
  n14 = VertexTangentWS()
  n15 = CrossProduct(A=@n14, B=VertexNormalWS())
  n16 = FunctionInput() {bUsePreviewValueAsDefault=true, inputName=WPOUseBiNormal, inputType=FunctionInput_Scalar, sortPriority=25}
  n17 = Normalize(VectorInput=Transform(None=LinearInterpolate(A=@n14, B=@n15, Alpha=@n16)) {transformSourceType=TRANSFORMSOURCE_World, transformType=TRANSFORM_Instance})
  n18 = FunctionInput() {bUsePreviewValueAsDefault=true, inputName=WPOStrength, inputType=FunctionInput_Scalar, sortPriority=25}
  FunctionOutput(None=Transform(None=Multiply(A=@n17, B=@n18)) {transformSourceType=TRANSFORMSOURCE_Instance}) {outputName=WPOWS, sortPriority=10}
  n19 = Saturate(None=@n4)
  n20 = FunctionInput() {bUsePreviewValueAsDefault=true, inputName=TipAreaOffset, inputType=FunctionInput_Scalar, sortPriority=7, previewValue=<10 0 0 1>}
  n21 = FunctionInput() {bUsePreviewValueAsDefault=true, inputName=TipAreaWidth_Inverse, inputType=FunctionInput_Scalar, sortPriority=6, previewValue=<10 0 0 1>}
  n22 = OneMinus(None=Multiply(A=Add(A=@n19, B=@n20), B=@n21))
  n23 = FunctionInput() {bUsePreviewValueAsDefault=true, inputName=TipAreaPower, inputType=FunctionInput_Scalar, sortPriority=8, previewValue=<1 0 0 1>}
  FunctionOutput(None=Power(Base=Saturate(None=@n22), Exp=@n23)) {outputName=TipArea, sortPriority=1}
 ]}
