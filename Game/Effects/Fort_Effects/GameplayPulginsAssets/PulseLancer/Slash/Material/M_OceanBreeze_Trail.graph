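# Additive, unlit, two-sided material for a PulseLancer slash trail, flagged for Niagara ribbons and mesh particles.
# ProceduralSlashParameters builds growth and dissolve masks from a looping S-curve timer that blends toward Dynamic Parameter inputs.
# Opacity multiplies sine-based stripes, edge masks, particle alpha, squared growth and squared depth fade, with an optional segmented digital mask.
# Color adds a scaled, desaturation-adjusted rainbow texture to particle-tinted base color, then applies a post multiplier.
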Material /Game/Effects/Fort_Effects/GameplayPulginsAssets/PulseLancer/Slash/Material/M_OceanBreeze_Trail
graph {
 properties {
  bUsedWithNanite=true
  bUsedWithNiagaraMeshParticles=true
  bUsedWithNiagaraRibbons=true
  blendMode=BLEND_Additive
  pixelDepthOffsetMode=PDOM_Legacy
  refractionMethod=RM_PixelNormalOffset
  shadingModel=MSM_Unlit
  twoSided=true
  usedShadingModels=Unlit
 }
 nodes [
  n1 = Frac(None=Multiply(A=Time(), B=ScalarParameter() {defaultValue=1, parameterName=ST}))
  n2 = ScalarParameter() {defaultValue=0.9, parameterName=teat}
  n3 = DynamicParameter() {defaultValue=<rgba 1 1 1 0>, paramNames [GrowthInput,Dissolve,WPO,EnableDP]}
  n4 = LinearInterpolate(
   A=MaterialFunctionCall(
    Input=MaterialFunctionCall(In=@n1, Power=@n2) {materialFunction=&SCurve}.Result
    Target High=ScalarParameter() {defaultValue=3.683, parameterName="GrowthTime0-1 (S)"}
   ) {materialFunction=&RemapValueRange}.Result
   B=@n3.GrowthInput
   Alpha=@n3.GrowthInput
  )
  n5 = ScalarParameter() {defaultValue=-0.6272, parameterName="DissolveTime0-1 (S)"}
  n6 = MaterialFunctionCall!(
   GrowthTime0-1=@n4
   GrowthBlendWidth_Inverse=ScalarParameter() {defaultValue=10, parameterName="GrowthBlendWidth_Inverse (S)"}
   GrowthBlendPower=ScalarParameter() {defaultValue=2, parameterName="GrowthBlendPower (S)"}
   ---------------
   TipAreaWidth_Inverse=ScalarParameter() {defaultValue=1.2, parameterName="TipAreaWidth_Inverse (S)"}
   TipAreaOffset=ScalarParameter() {defaultValue=-0.0768, parameterName="TipAreaOffset (S)"}
   TipAreaPower=ScalarParameter() {defaultValue=3, parameterName="TipAreaPower (S)"}
   ----------------
   DissolveTime0-1=LinearInterpolate(A=Add(A=@n4, B=@n5), B=@n3.Dissolve, Alpha=@n3.Dissolve)
   DissolveBlendWidth_Inverse=ScalarParameter() {defaultValue=1, parameterName="DissolveBlendWidth_Inverse (S)"}
   DissolveBlendPower=ScalarParameter() {defaultValue=2, parameterName="DissolveBlendPower (S)"}
   -----------------
   WPOUseBiNormal
   WPOStrength=@n3.WPO
  ) {materialFunction=/Game/Effects/Fort_Effects/GameplayPulginsAssets/PulseLancer/Slash/ProceduralSlashParameters}
  n7 = VectorParameter() {parameterName=TipArea_Refine, defaultValue=<rgba 8 532.2567 20 0.5>}
  Add(A=MaterialFunctionCall(Texture=@n6.TipArea, RGB=@n7.RGB, A=@n7.RGB) {materialFunction=&MF_FX_Refine}.Result)
  n8 = MaterialFunctionCall(Float2=TextureCoordinate()) {materialFunction=&BreakOutFloat2Components}
  n9 = Multiply(A=@n8.G) {constB=2}
  n10 = Saturate(None=@n9)
  n11 = OneMinus(None=Saturate(None=Subtract(A=@n9)))
  n12 = Ceil(None=Subtract(A=@n9))
  n13 = LinearInterpolate(A=@n8.G, B=OneMinus(None=LinearInterpolate(A=@n10, B=@n11, Alpha=@n12)), Alpha=ScalarParameter() {parameterName=Use2xV})
  n14 = VectorParameter() {
   group=OpacityMask
   parameterName=OpacityMask_EdgeMask_InvertControl
   channelNames {a="", b=Power, g=Multiply, r=Offset}
   defaultValue=<rgba 0 1 1 1>
  }
  n15 = VectorParameter() {
   group=OpacityMask
   parameterName=OpacityMask_EdgeMask_Control
   channelNames {a="", b=Power, g=Multiply, r=Offset}
   defaultValue=<rgba 0 10 2 1>
  }
  n16 = VectorParameter() {parameterName=GridUV, defaultValue=<rgba 20 0.05 0 1>}
  n17 = Multiply(A=TextureCoordinate() {coordinateIndex=1}, B=AppendVector(A=@n16.R, B=@n16.R))
  n18 = Sine(None=MaterialFunctionCall(Float2=Add(A=@n17) {constB=0.01}) {materialFunction=&BreakOutFloat2Components}.R)
  n19 = VectorParameter() {parameterName=GridRemap, defaultValue=<rgba -0.8 1 1 0.5>}
  MaterialExpressionRemap_0 = MaterialXRemap!(None=@n18, InputLow=@n19.R, InputHigh=@n19.R, TargetLow=@n19.R, TargetHigh=@n19.R)
  n20 = VectorParameter() {parameterName=GridOpacityOffset, defaultValue=<rgba 1.4 0.7 0 1>}
  n21 = Saturate(None=MaterialFunctionCall(In=Add(A=@n18, B=@n20.R), Contrast=ScalarParameter() {defaultValue=0.8, parameterName=GridContrast}) {materialFunction=&CheapContrast}.Result)
  n22 = VectorParameter() {parameterName=BaseLayerColor, defaultValue=<rgba 0 0 0 0.7>}
  n23 = Multiply(
   A=Saturate(None=MaterialFunctionCall!(UorV=@n13, Invert_EdgeControl=@n14, NoInvert_EdgeControl=@n15) {materialFunction=/Game/Effects/Niagara/MasterMaterials/Functions/Common/FXCommon_EdgeMask_Ex}.Mask)
   B=Multiply(A=Multiply(A=@MaterialExpressionRemap_0, B=@n21) {constA=1}, B=@n22.A)
  )
  n24 = If(A=@n23, "A > B"=@n23, "A == B"=@n23, "A < B"=Constant()) {constB=0.01}
  n25 = ParticleColor()
  n26 = Power(Base=@n25.A) {constExponent=1.3}
  n27 = Multiply(A=@n6.Growth, B=@n6.Growth)
  LinearInterpolate(
   B=LinearInterpolate(B=ScalarParameter() {defaultValue=0.99, parameterName=Refraction}, Alpha=LinearInterpolate(B=@n24, Alpha=@n26)) {constA=1}
   Alpha=@n27
  ) {constA=1}
  n28 = TextureObjectParameter() {group=RainbowColor, parameterName=RainbowColor_Texture, sortPriority=0, texture=&T_Tile_RainbowColor}
  n29 = VectorParameter() {group=RainbowColor, parameterName=RainbowColor_UVControl, channelNames {a=VPan, b=UPan, g=VScale, r=UScale}, defaultValue=<rgba 2 0.3 0.1 0>}
  n30 = Multiply(
   A=Desaturation(
    None=MaterialFunctionCall(Texture=@n28, UVControl_RGB=@n29, UVControl_A=@n29) {materialFunction=&FXCommon_TextureSample}.RGB
    Fraction=ScalarParameter() {defaultValue=-0.36, parameterName=RainbowDes}
   )
   B=ScalarParameter() {defaultValue=0.5, parameterName=RainbowStrength}
  )
  n31 = Add(A=@n30, B=Multiply(A=@n25.RGB, B=@n22.RGB))
  n32 = VectorParameter() {parameterName=PostColorMulti, defaultValue=<rgba 1 1 1 2>}
  MaterialFunctionCall(Emissive=Divide(A=Multiply(A=@n31, B=@n32.RGB)) {constB=1.9771}) {materialFunction=&MF_FX_EmissiveAdjustments}
  n33 = Multiply(A=@n23, B=@n25.A)
  n34 = VectorParameter() {group=EdgeB, parameterName=EdgeB_EdgeMask_InvertControl, channelNames {a="", b=Power, g=Multiply, r=Offset}, defaultValue=<rgba 0 3 1.5 1>}
  n35 = VectorParameter() {group=EdgeB, parameterName=EdgeB_EdgeMask_Control, channelNames {a="", b=Power, g=Multiply, r=Offset}, defaultValue=<rgba 0 5 2 1>}
  n36 = MaterialFunctionCall(UorV=TextureCoordinate(), Invert_EdgeControl=@n34, NoInvert_EdgeControl=@n35) {materialFunction=&FXCommon_EdgeMask_Ex}
  n37 = Multiply(A=Multiply(A=Multiply(A=@n33, B=@n36.Mask), B=@n32.A), B=@n27)
  n38 = Saturate(None=@n37)
  n39 = MaterialFunctionCall!(InUV, MaskInput=DynamicParameter() {parameterIndex=3, defaultValue=<rgba 0 0.2 0 0>, paramNames [InputMask,Param2,Param3,Param4]}.InputMask) {materialFunction=/Game/Effects/Fort_Effects/GameplayPulginsAssets/PulseLancer/HeavyAttack/MF_Segmented}
  n40 = DepthFade(FadeDistance=ScalarParameter() {defaultValue=100, parameterName=DF})
  Multiply(
   A=LinearInterpolate(A=@n37, B=Multiply(A=@n38, B=@n39.Result), Alpha=ScalarParameter() {parameterName=EnableDigitalMask})
   B=Multiply(A=@n40, B=@n40)
  )
  Multiply(A=@n30, B=Power(Base=OneMinus(None=@n6.Dissolve)) {constExponent=1.5})
  ScalarParameter() {parameterName="WPOStrength (S)"}
  TextureCoordinate()
  VertexNormalWS()
 ]
}
