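# Additive, unlit, two-sided Niagara material for the Pulse Lancer slash effect's wave lines.
# ProceduralSlashParameters supplies growth, tip and world-offset values from dynamic parameters blended with a time loop, alongside a binormal width offset.
# Named reroutes build periodic sine lines, U-quantized noise detail, UV-distortion noise and hue-shift noise.
# Refined masks become hue-shifted emissive tinted by particle color, with opacity shaped by growth, edge masks and cubed depth fade.
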Material /Game/Effects/Fort_Effects/GameplayPulginsAssets/PulseLancer/Slash/Material/M_PL_WaveLine
graph {
 properties {
  bAutomaticallySetUsageInEditor=false
  bUsedWithNiagaraMeshParticles=true
  bUsedWithNiagaraRibbons=true
  bUsedWithNiagaraSprites=true
  blendMode=BLEND_Additive
  shadingModel=MSM_Unlit
  twoSided=true
  usedShadingModels=Unlit
 }
 nodes [
  n1 = Frac(None=Multiply(A=Time(), B=ScalarParameter() {defaultValue=1, parameterName=ST}))
  n2 = ScalarParameter() {defaultValue=0.9, parameterName=teat}
  n3 = DynamicParameter() {defaultValue=<rgba 1 1 1 0>, paramNames [GrowthInput,Dissolve,Param3,EnableDP]}
  n4 = LinearInterpolate(
   A=MaterialFunctionCall(
    Input=MaterialFunctionCall(In=@n1, Power=@n2) {materialFunction=&SCurve}.Result
    Target High=ScalarParameter() {defaultValue=3.683, parameterName="GrowthTime0-1 (S)"}
   ) {materialFunction=&RemapValueRange}.Result
   B=@n3.GrowthInput
   Alpha=@n3.GrowthInput
  )
  n5 = ScalarParameter() {defaultValue=-0.8862, parameterName="DissolveTime0-1 (S)"}
  n6 = MaterialFunctionCall!(
   GrowthTime0-1=@n4
   GrowthBlendWidth_Inverse=ScalarParameter() {defaultValue=10, parameterName="GrowthBlendWidth_Inverse (S)"}
   GrowthBlendPower=ScalarParameter() {defaultValue=2, parameterName="GrowthBlendPower (S)"}
   ---------------
   TipAreaWidth_Inverse=ScalarParameter() {defaultValue=1.2, parameterName="TipAreaWidth_Inverse (S)"}
   TipAreaOffset=ScalarParameter() {defaultValue=-0.0768, parameterName="TipAreaOffset (S)"}
   TipAreaPower=ScalarParameter() {defaultValue=3, parameterName="TipAreaPower (S)"}
   ----------------
   DissolveTime0-1=LinearInterpolate(A=Add(A=@n4, B=@n5), B=@n3.Dissolve, Alpha=@n3.Dissolve)
   DissolveBlendWidth_Inverse=ScalarParameter() {defaultValue=1, parameterName="DissolveBlendWidth_Inverse (S)"}
   DissolveBlendPower=ScalarParameter() {defaultValue=2, parameterName="DissolveBlendPower (S)"}
   -----------------
   WPOUseBiNormal
   WPOStrength=@n3.Param3
  ) {materialFunction=/Game/Effects/Fort_Effects/GameplayPulginsAssets/PulseLancer/Slash/ProceduralSlashParameters}
  n7 = VectorParameter() {parameterName=TipArea_Refine, defaultValue=<rgba 8 532.2567 20 0.5>}
  Add(A=MaterialFunctionCall(Texture=@n6.TipArea, RGB=@n7.RGB, A=@n7.RGB) {materialFunction=&MF_FX_Refine}.Result)
  n8 = CrossProduct(A=VertexNormalWS(), B=VertexTangentWS())
  n9 = Subtract(A=MaterialFunctionCall(Float2=TextureCoordinate()) {materialFunction=&BreakOutFloat2Components}.G) {constB=0.5}
  Add(A=@n6.WPOWS, B=Multiply(A=Multiply(A=@n8, B=@n9), B=ScalarParameter() {parameterName=TS}))
  DynamicParameter() {parameterIndex=1, defaultValue=<rgba 1 1 1 0>, paramNames [SandStrength,Param2,Param3,Param4]}
  n10 = MaterialFunctionCall(Float2=TextureCoordinate()) {materialFunction=&BreakOutFloat2Components}
  n11 = VectorParameter() {
   group=SandFade
   parameterName=SandFade_EdgeMask_InvertControl
   channelNames {a="", b=Power, g=Multiply, r=Offset}
   defaultValue=<rgba 0 2 1.5 1>
  }
  n12 = VectorParameter() {
   group=SandFade
   parameterName=SandFade_EdgeMask_Control
   channelNames {a="", b=Power, g=Multiply, r=Offset}
   defaultValue=<rgba 0 375.0543 5 1>
  }
  n13 = VectorParameter() {group=HueT2, parameterName=HueT2_UVControl, channelNames {a=VPan, b=UPan, g=VScale, r=UScale}, defaultValue=<rgba 2 2 0 0>}
  n14 = Multiply(
   A=MaterialFunctionCall(
    Texture=TextureObjectParameter() {group=HueT2, parameterName=HueT2_Texture, sortPriority=0, texture=/Game/Global/Textures/Default/Noise/T_Tiling_Noise_01}
    UVControl_RGB=@n13
    UVControl_A=@n13
   ) {materialFunction=&FXCommon_TextureSample}.R
   B=ScalarParameter() {defaultValue=1.5, parameterName=HS}
  )
  n15 = NamedRerouteUsage()
  n16 = VectorParameter() {parameterName=LineDetail_Refine2, defaultValue=<rgba 3 231.3944 1 0.5>}
  n17 = Power(Base=NamedRerouteUsage(), Exp=ScalarParameter() {defaultValue=3, parameterName=FB2})
  n18 = VectorParameter() {parameterName=Line_Refine2, defaultValue=<rgba 30.623 5 0.2 0.5>}
  n19 = Multiply(
   A=Multiply(
    A=MaterialFunctionCall(Texture=@n15, RGB=@n16.RGB, A=@n16.RGB) {materialFunction=&MF_FX_Refine}.Result
    B=MaterialFunctionCall(Texture=@n17, RGB=@n18.RGB, A=@n18.RGB) {materialFunction=&MF_FX_Refine}.Result
   )
   B=VectorParameter() {parameterName=STC, defaultValue=<rgba 0.0825 0.0995 0.5 1>}.RGB
  )
  n20 = VectorParameter() {parameterName=TestC, defaultValue=<rgba 0.56 0.68 2 0.2>}
  n21 = NamedRerouteUsage()
  n22 = Add(
   A=Multiply(
    A=MaterialFunctionCall!(UorV=@n10.R, Invert_EdgeControl=@n11, NoInvert_EdgeControl=@n12) {materialFunction=/Game/Effects/Niagara/MasterMaterials/Functions/Common/FXCommon_EdgeMask_Ex}.Mask
    B=MaterialFunctionCall(Hue Shift Percentage=@n14, Texture=@n19) {materialFunction=&HueShift}.Result
   )
   B=MaterialFunctionCall(Hue Shift Percentage=Multiply(A=@n20.A, B=@n21), Texture=@n20.RGB) {materialFunction=&HueShift}.Result
  )
  n23 = ParticleColor()
  MaterialFunctionCall(Emissive=Divide(A=Multiply(A=@n22, B=@n23.RGB)) {constB=1.97}) {materialFunction=&MF_FX_EmissiveAdjustments}
  n24 = DepthFade(FadeDistance=ScalarParameter() {defaultValue=100, parameterName=DF})
  n25 = NamedRerouteUsage()
  n26 = ScalarParameter() {parameterName=Fw3}
  n27 = Saturate(None=Multiply(A=Subtract(A=@n25, B=@n26), B=ScalarParameter() {defaultValue=1.01, parameterName=32t}))
  n28 = ScalarParameter() {defaultValue=5, parameterName=FB}
  n29 = VectorParameter() {parameterName=Line_Refine, defaultValue=<rgba 9.4151 1.2239 0.5 0.5>}
  n30 = Multiply(A=@n6.Growth, B=@n6.Growth)
  n31 = MaterialFunctionCall(Float2=TextureCoordinate()) {materialFunction=&BreakOutFloat2Components}
  n32 = VectorParameter() {
   group=TipTailFade
   parameterName=TipTailFade_EdgeMask_InvertControl
   channelNames {a="", b=Power, g=Multiply, r=Offset}
   defaultValue=<rgba 0 2 2 1>
  }
  n33 = VectorParameter() {
   group=TipTailFade
   parameterName=TipTailFade_EdgeMask_Control
   channelNames {a="", b=Power, g=Multiply, r=Offset}
   defaultValue=<rgba 0 897.498 2 1>
  }
  n34 = VectorParameter() {
   group=SideFade
   parameterName=SideFade_EdgeMask_InvertControl
   channelNames {a="", b=Power, g=Multiply, r=Offset}
   defaultValue=<rgba 0 2 2.5 1>
  }
  n35 = VectorParameter() {
   group=SideFade
   parameterName=SideFade_EdgeMask_Control
   channelNames {a="", b=Power, g=Multiply, r=Offset}
   defaultValue=<rgba 0 5.1445 3.0882 1>
  }
  n36 = Multiply(
   A=MaterialFunctionCall(UorV=@n31.R, Invert_EdgeControl=@n32, NoInvert_EdgeControl=@n33) {materialFunction=&FXCommon_EdgeMask_Ex}.Mask
   B=MaterialFunctionCall(UorV=@n31.G, Invert_EdgeControl=@n34, NoInvert_EdgeControl=@n35) {materialFunction=&FXCommon_EdgeMask_Ex}.Mask
  )
  n37 = Multiply(
   A=MaterialFunctionCall(Texture=Power(Base=@n27, Exp=@n28), RGB=@n29.RGB, A=@n29.RGB) {materialFunction=&MF_FX_Refine}.Result
   B=Multiply(A=Multiply(A=@n30, B=@n36))
  )
  Multiply(A=Multiply(A=Multiply(A=@n24, B=@n24), B=@n24), B=Saturate(None=Multiply(A=@n37, B=@n23.A)))
  n38 = TextureObjectParameter() {
   group=HueShift
   parameterName=HueShift_Texture
   samplerType=SAMPLERTYPE_LinearGrayscale
   sortPriority=0
   texture=/Game/Effects/Fort_Effects/Emotes/UltraLight/Texture/T_Tiling_Noise_01_GreyScale
  }
  n39 = VectorParameter() {group=HueShift, parameterName=HueShift_UVControl, channelNames {a=VPan, b=UPan, g=VScale, r=UScale}, defaultValue=<rgba 2 0.5 0.5 0>}
  NamedRerouteDeclaration(None=Add(
    A=MaterialFunctionCall(Texture=@n38, UVControl_RGB=@n39, UVControl_A=@n39) {materialFunction=&FXCommon_TextureSample}.R
    B=ScalarParameter() {defaultValue=-0.5, parameterName=ssHAE}
   ) {constB=-0.5}) {name=HueShiftTexture, nodeColor=<rgba 0 1 0.3438 1>}
  n40 = Add(A=TextureCoordinate(), B=Multiply(A=Time(), B=ScalarParameter() {defaultValue=0.05, parameterName=t}))
  n41 = NamedRerouteUsage()
  NamedRerouteDeclaration(None=MaterialFunctionCall(Value=Add(A=@n40, B=@n41), Period=ScalarParameter() {defaultValue=0.04, parameterName=P}, -1 to 1=StaticBool()) {materialFunction=&LinearSine}.Linear Sine) {name=SineLinear, nodeColor=<rgba 1 0.0234 0 1>}
  n42 = VectorParameter() {
   group=LineDetail
   parameterName=LineDetail_UVControl
   channelNames {a=VPan, b=UPan, g=VScale, r=UScale}
   defaultValue=<rgba 7.8196 1.5 0 0.2>
  }
  n43 = MaterialFunctionCall(Float2=TextureCoordinate()) {materialFunction=&BreakOutFloat2Components}
  n44 = ScalarParameter() {defaultValue=0.04, parameterName=P}
  n45 = Add(A=@n43.R, B=Multiply(A=@n44) {constB=-0.75}) {constB=0}
  n46 = Multiply(A=Floor(None=Divide(A=@n45, B=@n44)), B=@n44)
  NamedRerouteDeclaration(None=MaterialFunctionCall(
    Texture=TextureObjectParameter() {group=LineDetail, parameterName=LineDetail_Texture, sortPriority=0, texture=/Game/Global/Textures/Default/Noise/Noise}
    AdditionalUVOffset=NamedRerouteUsage()
    UVControl_RGB=@n42
    UVControl_A=@n42
    "[Override]UVInput"=AppendVector(A=@n46, B=@n43.G)
   ) {materialFunction=&FXCommon_TextureSample}.R) {name=SandT, nodeColor=<rgba 0 0.2969 1 1>}
  n47 = TextureObjectParameter() {
   group=ShapeNoise
   parameterName=ShapeNoise_Texture
   samplerType=SAMPLERTYPE_Normal
   sortPriority=0
   texture=/Game/Characters/Player/Female/Medium/Bodies/F_MED_Bistro_Astronaut/FX/T_GaussianNoise_Normal
  }
  n48 = VectorParameter() {group=ShapeNoise, parameterName=ShapeNoise_UVControl, channelNames {a=VPan, b=UPan, g=VScale, r=UScale}, defaultValue=<rgba 5 0.5707 0 0.1>}
  n49 = VectorParameter() {
   group=ShapeNoise
   parameterName=ShapeNoise_NoiseControl
   sortPriority=60
   channelNames {a=VStrength, b=UStrength, g=VOffset, r=UOffset}
   defaultValue=<rgba -0.5 -0.5 0 0>
  }
  NamedRerouteDeclaration(None=MaterialFunctionCall!(
    NoiseValue=MaterialFunctionCall(Texture=@n47, UVControl_RGB=@n48, UVControl_A=@n48) {materialFunction=&FXCommon_TextureSample}.RG
    Control_RGB=@n49
    Control_A=@n49
   ) {materialFunction=/Game/Effects/Niagara/MasterMaterials/Functions/Common/FXCommon_NoiseControl}.NoiseValue) {name=ShapeNoise, nodeColor=<rgba 0.5 0 1 1>}
  n50 = Subtract(A=VertexColor().R, B=DynamicParameter() {defaultValue=<rgba 1 1 1 0>, paramNames [GrowthInput,Dissolve,Param3,EnableDP]}.Dissolve)
  n51 = Saturate(None=Multiply(A=Multiply(A=@n50) {constB=-1}, B=ScalarParameter() {defaultValue=1.1, parameterName=DisappearScale}))
  n52 = Multiply(A=OneMinus(None=@n51), B=Ceil(None=@n51))
  NamedRerouteDeclaration(None=Power(Base=Saturate(None=@n52), Exp=ScalarParameter() {defaultValue=2, parameterName=DIsPower})) {name=Dis, nodeColor=<rgba 0.9688 0 1 1>}
  n53 = OneMinus(None=MaterialFunctionCall(Float2=TextureCoordinate()) {materialFunction=&BreakOutFloat2Components}.R)
  n54 = DynamicParameter() {defaultValue=<rgba 1 1 1 0>, paramNames [GrowthInput,Dissolve,Param3,EnableDP]}
  n55 = Multiply(
   A=Add(A=Subtract(A=@n53, B=@n54.Param3), B=ScalarParameter() {group=FakeDis, parameterName=FakeDis_GrayAdjust_Add})
   B=ScalarParameter() {defaultValue=1, group=FakeDis, parameterName=FakeDis_GrayAdjust_Mul}
  )
  Saturate(None=Power(Base=Saturate(None=@n55), Exp=ScalarParameter() {defaultValue=2, group=FakeDis, parameterName=FakeDis_GrayAdjust_Power}))
  ScalarParameter() {parameterName="WPOStrength (S)"}
 ]
}
