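# M_SlashDark is a translucent, unlit, two-sided Niagara material for the dark, smoky slash effect in PulseLancer.
# A looping S-curved time value, blended with the GrowthInput dynamic parameter by that same input, drives growth and dissolve in the ProceduralSlashParameters function.
# An inky smoke texture is remapped and tinted dark.
# Its powered value is masked by squared growth and tip/tail and side edge fades.
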Material /Game/Effects/Fort_Effects/GameplayPulginsAssets/PulseLancer/Slash/Material/M_SlashDark
graph {
 properties {
  bAutomaticallySetUsageInEditor=false
  bUsedWithNiagaraMeshParticles=true
  bUsedWithNiagaraRibbons=true
  bUsedWithNiagaraSprites=true
  blendMode=BLEND_Translucent
  shadingModel=MSM_Unlit
  twoSided=true
  usedShadingModels=Unlit
 }
 nodes [
  n1 = Frac(None=Multiply(A=Time(), B=ScalarParameter() {defaultValue=1, parameterName=ST}))
  n2 = ScalarParameter() {defaultValue=0.9, parameterName=teat}
  n3 = DynamicParameter() {defaultValue=<rgba 1 1 1 0>, paramNames [GrowthInput,Param2,Param3,EnableDP]}
  n4 = LinearInterpolate(
   A=MaterialFunctionCall(
    Input=MaterialFunctionCall(In=@n1, Power=@n2) {materialFunction=&SCurve}.Result
    Target High=ScalarParameter() {defaultValue=3.683, parameterName="GrowthTime0-1 (S)"}
   ) {materialFunction=&RemapValueRange}.Result
   B=@n3.GrowthInput
   Alpha=@n3.GrowthInput
  )
  n5 = MaterialFunctionCall!(
   GrowthTime0-1=@n4
   GrowthBlendWidth_Inverse=ScalarParameter() {defaultValue=10, parameterName="GrowthBlendWidth_Inverse (S)"}
   GrowthBlendPower=ScalarParameter() {defaultValue=2, parameterName="GrowthBlendPower (S)"}
   ---------------
   TipAreaWidth_Inverse=ScalarParameter() {defaultValue=1.2, parameterName="TipAreaWidth_Inverse (S)"}
   TipAreaOffset=ScalarParameter() {defaultValue=-0.0768, parameterName="TipAreaOffset (S)"}
   TipAreaPower=ScalarParameter() {defaultValue=3, parameterName="TipAreaPower (S)"}
   ----------------
   DissolveTime0-1=Add(A=@n4, B=ScalarParameter() {defaultValue=-0.96, parameterName="DissolveTime0-1 (S)"})
   DissolveBlendWidth_Inverse=ScalarParameter() {defaultValue=1, parameterName="DissolveBlendWidth_Inverse (S)"}
   DissolveBlendPower=ScalarParameter() {defaultValue=2, parameterName="DissolveBlendPower (S)"}
   -----------------
   WPOUseBiNormal
   WPOStrength=ScalarParameter() {parameterName="WPOStrength (S)"}
  ) {materialFunction=/Game/Effects/Fort_Effects/GameplayPulginsAssets/PulseLancer/Slash/ProceduralSlashParameters}
  n6 = VectorParameter() {parameterName=TipArea_Refine, defaultValue=<rgba 8 532.2567 20 0.5>}
  Add(A=MaterialFunctionCall(Texture=@n5.TipArea, RGB=@n6.RGB, A=@n6.RGB) {materialFunction=&MF_FX_Refine}.Result)
  n7 = TextureObjectParameter() {
   group=MainTexture
   parameterName=MainTexture_Texture
   sortPriority=0
   texture=&T_Inky_Smoke_Tile
  }
  n8 = VectorParameter() {group=MainTexture, parameterName=MainTexture_UVControl, channelNames {a=VPan, b=UPan, g=VScale, r=UScale}, defaultValue=<rgba 1 2 0 0>}
  n9 = MaterialFunctionCall(Texture=@n7, UVControl_RGB=@n8, UVControl_A=@n8) {materialFunction=&FXCommon_TextureSample}
  Multiply(
   A=MaterialFunctionCall(Input=@n9.RGB, Target Low=ScalarParameter() {defaultValue=0.3, parameterName=TH}) {materialFunction=&RemapValueRange}.Result
   B=VectorParameter() {parameterName=DarkC, defaultValue=<rgba 0.4 0.4035 0.5 1>}.RGB
  )
  n10 = Multiply(A=@n5.Growth, B=@n5.Growth)
  n11 = MaterialFunctionCall(Float2=TextureCoordinate()) {materialFunction=&BreakOutFloat2Components}
  n12 = VectorParameter() {
   group=TipTailFade
   parameterName=TipTailFade_EdgeMask_InvertControl
   channelNames {a="", b=Power, g=Multiply, r=Offset}
   defaultValue=<rgba 0 2 1 1>
  }
  n13 = VectorParameter() {
   group=TipTailFade
   parameterName=TipTailFade_EdgeMask_Control
   channelNames {a="", b=Power, g=Multiply, r=Offset}
   defaultValue=<rgba 0 10 2 1>
  }
  n14 = VectorParameter() {
   group=SideFade
   parameterName=SideFade_EdgeMask_InvertControl
   channelNames {a="", b=Power, g=Multiply, r=Offset}
   defaultValue=<rgba 0 1 1 1>
  }
  n15 = VectorParameter() {group=SideFade, parameterName=SideFade_EdgeMask_Control, channelNames {a="", b=Power, g=Multiply, r=Offset}, defaultValue=<rgba 0 150 2 1>}
  n16 = Multiply(
   A=MaterialFunctionCall!(UorV=@n11.R, Invert_EdgeControl=@n12, NoInvert_EdgeControl=@n13) {materialFunction=/Game/Effects/Niagara/MasterMaterials/Functions/Common/FXCommon_EdgeMask_Ex}.Mask
   B=MaterialFunctionCall(UorV=@n11.G, Invert_EdgeControl=@n14, NoInvert_EdgeControl=@n15) {materialFunction=&FXCommon_EdgeMask_Ex}.Mask
  )
  Multiply(A=Power(Base=@n9.RGB, Exp=ScalarParameter() {defaultValue=0.2, parameterName=awd}), B=Multiply(A=Multiply(A=@n10, B=@n16)))
 ]
}
