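# Unlit, two-sided translucent material for PulseLancer slash trails on Niagara ribbons, meshes and sprites.
# ProceduralSlashParameters shapes growth, dissolve and tip masks from Dynamic Parameter inputs, blended with a looping time-driven fallback.
# A noise-offset plasma texture, tinted by noise-driven hue shift, tip highlight and particle color, builds the emissive.
# Opacity multiplies squared growth, texture luminance, noise-warped tip/tail and side fades, particle alpha and cubed depth fade.
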Material /Game/Effects/Fort_Effects/GameplayPulginsAssets/PulseLancer/Slash/Material/M_SoftSlashTrail
graph {
 properties {
  allowTranslucentCustomDepthWrites=true
  bAutomaticallySetUsageInEditor=false
  bEnableResponsiveAA=true
  bUsedWithNiagaraMeshParticles=true
  bUsedWithNiagaraRibbons=true
  bUsedWithNiagaraSprites=true
  blendMode=BLEND_Translucent
  ditherOpacityMask=true
  opacityMaskClipValue=0.001
  shadingModel=MSM_Unlit
  twoSided=true
  usedShadingModels=Unlit
 }
 nodes [
  n1 = MaterialFunctionCall(Float2=TextureCoordinate()) {materialFunction=&BreakOutFloat2Components}
  n2 = VectorParameter() {group=Fade2, parameterName=Fade2_EdgeMask_Control, channelNames {a="", b=Power, g=Multiply, r=Offset}, defaultValue=<rgba 1 50.6575 1 1>}
  MaterialFunctionCall!(UorV=@n1.R, EdgeControl=@n2) {materialFunction=/Game/Effects/Niagara/MasterMaterials/Functions/Common/FXCommon_EdgeMask_Simple}
  n3 = Frac(None=Multiply(A=Time(), B=ScalarParameter() {defaultValue=1, parameterName=ST}))
  n4 = ScalarParameter() {defaultValue=0.9, parameterName=teat}
  n5 = DynamicParameter() {defaultValue=<rgba 1 1 1 0>, paramNames [GrowthInput, DissolveInput, WPO, EnableDP]}
  n6 = LinearInterpolate(
   A=MaterialFunctionCall(
    Input=MaterialFunctionCall(In=@n3, Power=@n4) {materialFunction=&SCurve}.Result
    Target High=ScalarParameter() {defaultValue=3.683, parameterName="GrowthTime0-1 (S)"}
   ) {materialFunction=&RemapValueRange}.Result
   B=@n5.GrowthInput
   Alpha=@n5.GrowthInput
  )
  n7 = ScalarParameter() {defaultValue=-0.96, parameterName="DissolveTime0-1 (S)"}
  n8 = MaterialFunctionCall!(
   GrowthTime0-1=@n6
   GrowthBlendWidth_Inverse=ScalarParameter() {defaultValue=10, parameterName="GrowthBlendWidth_Inverse (S)"}
   GrowthBlendPower=ScalarParameter() {defaultValue=2, parameterName="GrowthBlendPower (S)"}
   ---------------
   TipAreaWidth_Inverse=ScalarParameter() {defaultValue=1.2, parameterName="TipAreaWidth_Inverse (S)"}
   TipAreaOffset=ScalarParameter() {defaultValue=-0.0768, parameterName="TipAreaOffset (S)"}
   TipAreaPower=ScalarParameter() {defaultValue=3, parameterName="TipAreaPower (S)"}
   ----------------
   DissolveTime0-1=LinearInterpolate(A=Add(A=@n6, B=@n7), B=@n5.DissolveInput, Alpha=@n5.DissolveInput)
   DissolveBlendWidth_Inverse=ScalarParameter() {defaultValue=1, parameterName="DissolveBlendWidth_Inverse (S)"}
   DissolveBlendPower=ScalarParameter() {defaultValue=1, parameterName="DissolveBlendPower (S)"}
   -----------------
   WPOUseBiNormal
   WPOStrength=@n5.WPO
  ) {materialFunction=/Game/Effects/Fort_Effects/GameplayPulginsAssets/PulseLancer/Slash/ProceduralSlashParameters}
  n9 = VectorParameter() {parameterName=TipArea_Refine, defaultValue=<rgba 9.0182 900 22.5455 0.5>}
  n10 = MaterialFunctionCall(Texture=@n8.TipArea, RGB=@n9.RGB, A=@n9.RGB) {materialFunction=&MF_FX_Refine}
  n11 = TextureObjectParameter() {
   group=MainTexture
   parameterName=MainTexture_Texture
   samplerType=SAMPLERTYPE_LinearColor
   sortPriority=0
   texture=/Game/Effects/Fort_Effects/Emotes/UltraLight/Texture/T_Plasma_A_C
  }
  n12 = TextureObjectParameter() {
   group=MainNoise
   parameterName=MainNoise_Texture
   sortPriority=0
   texture=/Game/Weapons/FORT_Melee/Pickaxe_CoatCheck/FX/T_Pickaxe_CoatCheck_Noise
  }
  n13 = VectorParameter() {group=MainNoise, parameterName=MainNoise_UVControl, channelNames {a=VPan, b=UPan, g=VScale, r=UScale}, defaultValue=<rgba 1 1 0 0>}
  n14 = VectorParameter() {
   group=MainNoise
   parameterName=MainNoise_NoiseControl
   sortPriority=60
   channelNames {a=VStrength, b=UStrength, g=VOffset, r=UOffset}
   defaultValue=<rgba 0 0 0.2 0>
  }
  n15 = MaterialFunctionCall!(
   NoiseValue=MaterialFunctionCall(Texture=@n12, UVControl_RGB=@n13, UVControl_A=@n13) {materialFunction=&FXCommon_TextureSample}.RG
   Control_RGB=@n14
   Control_A=@n14
  ) {materialFunction=/Game/Effects/Niagara/MasterMaterials/Functions/Common/FXCommon_NoiseControl}
  n16 = VectorParameter() {
   group=MainTexture
   parameterName=MainTexture_UVControl
   channelNames {a=VPan, b=UPan, g=VScale, r=UScale}
   defaultValue=<rgba -0.3125 3 0.5 0>
  }
  n17 = Desaturation(
   None=MaterialFunctionCall(Texture=@n11, AdditionalUVOffset=@n15.NoiseValue, UVControl_RGB=@n16, UVControl_A=@n16) {materialFunction=&FXCommon_TextureSample}.RGB
   Fraction=ScalarParameter() {defaultValue=-0.2, parameterName=DesatCA}
  )
  n18 = Length(None=@n17) {is=class MaterialExpressionLength}
  n19 = VectorParameter() {parameterName=MainTexture_Refine, defaultValue=<rgba 5 13.5879 0 0.5>}
  n20 = MaterialFunctionCall(Texture=@n18, RGB=@n19.RGB, A=@n19.RGB) {materialFunction=&MF_FX_Refine}
  n21 = MaterialFunctionCall(Float2=TextureCoordinate()) {materialFunction=&BreakOutFloat2Components}
  n22 = VectorParameter() {
   group=RefineAreaSize
   parameterName=RefineAreaSize_EdgeMask_Control
   channelNames {a="", b=Power, g=Multiply, r=Offset}
   defaultValue=<rgba 0 1.8549 6.2627 1>
  }
  n23 = Multiply(
   A=Normalize(VectorInput=@n17)
   B=LinearInterpolate(
    B=Add(A=@n20.Result)
    Alpha=MaterialFunctionCall(UorV=@n21.G, EdgeControl=@n22) {materialFunction=&FXCommon_EdgeMask_Simple}.Mask
   ) {constA=1}
  )
  n24 = TextureObjectParameter() {
   group=HueShiftTexture
   parameterName=HueShiftTexture_Texture
   sortPriority=0
   texture=/Game/Effects/Fort_Effects/Emotes/UltraLight/Texture/T_SoftNoise_BC7
  }
  n25 = VectorParameter() {
   group=HueShiftTexture
   parameterName=HueShiftTexture_UVControl
   channelNames {a=VPan, b=UPan, g=VScale, r=UScale}
   defaultValue=<rgba 2 2 0.5 0>
  }
  n26 = MaterialFunctionCall(Texture=@n24, UVControl_RGB=@n25, UVControl_A=@n25) {materialFunction=&FXCommon_TextureSample}
  n27 = MaterialFunctionCall(
   Hue Shift Percentage=Multiply(A=Subtract(A=@n26.R) {constB=0.5}, B=ScalarParameter() {defaultValue=0.1, parameterName=HueShiftStrength})
   Texture=VectorParameter() {parameterName=MainColor, defaultValue=<rgba 1.2242 1 4 1>}.RGB
  ) {materialFunction=&HueShift}
  n28 = ParticleColor()
  n29 = Multiply(A=Multiply(A=Add(A=@n10.Result), B=Multiply(A=@n23, B=@n27.Result)), B=@n28.RGB)
  MaterialFunctionCall(Emissive=Multiply(A=ScalarParameter() {defaultValue=1, parameterName=ScaleColor}, B=Divide(A=@n29) {constB=1.97})) {materialFunction=&MF_FX_EmissiveAdjustments}
  Multiply(B=@n15.NoiseValue)
  n30 = DepthFade(FadeDistance=ScalarParameter() {defaultValue=100, parameterName=DF})
  n31 = Multiply(A=@n8.Growth, B=@n8.Growth)
  n32 = TextureObjectParameter() {
   group=TipTailShapeNoise
   parameterName=TipTailShapeNoise_Texture
   samplerType=SAMPLERTYPE_LinearColor
   sortPriority=0
   texture=/Engine/Functions/Engine_MaterialFunctions02/ExampleContent/Textures/OffsetNoiseDistanceFields_ForNormals
  }
  n33 = VectorParameter() {
   group=TipTailShapeNoise
   parameterName=TipTailShapeNoise_UVControl
   channelNames {a=VPan, b=UPan, g=VScale, r=UScale}
   defaultValue=<rgba 1 1 0.5 0>
  }
  n34 = MaterialFunctionCall(Texture=@n32, UVControl_RGB=@n33, UVControl_A=@n33) {materialFunction=&FXCommon_TextureSample}
  n35 = Power(Base=@n34.RG, Exp=ScalarParameter() {defaultValue=2, parameterName=afww3})
  n36 = VectorParameter() {
   group=TipTailShapeNoise
   parameterName=TipTailShapeNoise_NoiseControl
   sortPriority=60
   channelNames {a=VStrength, b=UStrength, g=VOffset, r=UOffset}
   defaultValue=<rgba 0 0 -0.2496 0>
  }
  n37 = MaterialFunctionCall(Float2=Add(
    A=MaterialFunctionCall(NoiseValue=@n35, Control_RGB=@n36, Control_A=@n36) {materialFunction=&FXCommon_NoiseControl}.NoiseValue
    B=TextureCoordinate()
   )) {materialFunction=&BreakOutFloat2Components}
  n38 = VectorParameter() {
   group=TipTailFade
   parameterName=TipTailFade_EdgeMask_InvertControl
   channelNames {a="", b=Power, g=Multiply, r=Offset}
   defaultValue=<rgba 0 4 2 1>
  }
  n39 = VectorParameter() {
   group=TipTailFade
   parameterName=TipTailFade_EdgeMask_Control
   channelNames {a="", b=Power, g=Multiply, r=Offset}
   defaultValue=<rgba 0 14 2 1>
  }
  n40 = MaterialFunctionCall!(UorV=@n37.R, Invert_EdgeControl=@n38, NoInvert_EdgeControl=@n39) {materialFunction=/Game/Effects/Niagara/MasterMaterials/Functions/Common/FXCommon_EdgeMask_Ex}
  n41 = VectorParameter() {group=SideFade, parameterName=SideFade_EdgeMask_Control, channelNames {a="", b=Power, g=Multiply, r=Offset}, defaultValue=<rgba 0 2 2 1>}
  n42 = Multiply(
   A=MaterialFunctionCall(Input=@n18, Target Low=ScalarParameter() {defaultValue=0.1, parameterName=MainTextureRemapLow}) {materialFunction=&RemapValueRange}.Result
   B=Multiply(
    A=Multiply(A=@n31, B=@n40.Mask)
    B=MaterialFunctionCall(UorV=@n37.G, EdgeControl=@n41) {materialFunction=&FXCommon_EdgeMask_Simple}.Mask
   )
  )
  n43 = Multiply(A=@n42, B=@n28.A)
  n44 = VectorParameter() {group=Fade2, parameterName=Fade2_EdgeMask_InvertControl, channelNames {a="", b=Power, g=Multiply, r=Offset}, defaultValue=<rgba 0 50 1 1>}
  n45 = VectorParameter() {group=Fade2, parameterName=Fade2_EdgeMask_Control, channelNames {a="", b=Power, g=Multiply, r=Offset}, defaultValue=<rgba 1 50.6575 1 1>}
  Multiply(
   A=Multiply(A=Multiply(A=@n30, B=@n30), B=@n30)
   B=Multiply(
    A=Saturate(None=@n43)
    B=MaterialFunctionCall(UorV=@n1.R, Invert_EdgeControl=@n44, NoInvert_EdgeControl=@n45) {materialFunction=&FXCommon_EdgeMask_Ex}.Mask
   )
  )
  Multiply(
   A=MaterialFunctionCall!(ObjectOrParticlePosition=ObjectPositionWS()) {materialFunction=/Game/Effects/Fort_Effects/Materials/Toon/CustomDepthOutLine/MF_BackfaceCulling}.Result
   B=@n42
  )
  OneMinus(None=@n34.RG)
  ScalarParameter() {parameterName="WPOStrength (S)"}
 ]
}
