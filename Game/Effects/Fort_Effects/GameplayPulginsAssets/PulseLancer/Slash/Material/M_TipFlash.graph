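# Translucent, unlit, two-sided Niagara material for the tip flash of the PulseLancer slash effect.
# ProceduralSlashParameters drives growth, dissolve and tip masks from looping time, blended toward particle dynamic-parameter inputs.
# World position offset adds that function's offset to SplineThicken, whose width follows a tip-tail edge mask.
# Particle color tints an HDR emissive; opacity combines edge masks, squared growth, particle alpha and cubed depth fade.
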Material /Game/Effects/Fort_Effects/GameplayPulginsAssets/PulseLancer/Slash/Material/M_TipFlash
graph {
 properties {
  bAutomaticallySetUsageInEditor=false
  bUsedWithNiagaraMeshParticles=true
  bUsedWithNiagaraRibbons=true
  bUsedWithNiagaraSprites=true
  blendMode=BLEND_Translucent
  shadingModel=MSM_Unlit
  twoSided=true
  usedShadingModels=Unlit
 }
 nodes [
  n1 = Frac(None=Multiply(A=Time(), B=ScalarParameter() {defaultValue=1, parameterName=ST}))
  n2 = ScalarParameter() {defaultValue=0.9, parameterName=teat}
  n3 = DynamicParameter() {defaultValue=<rgba 1 1 1 0>, paramNames [GrowthInput, DissolveInput, Param3, EnableDP]}
  n4 = LinearInterpolate(
   A=MaterialFunctionCall(
    Input=MaterialFunctionCall(In=@n1, Power=@n2) {materialFunction=&SCurve}.Result
    Target High=ScalarParameter() {defaultValue=3.683, parameterName="GrowthTime0-1 (S)"}
   ) {materialFunction=&RemapValueRange}.Result
   B=@n3.GrowthInput
   Alpha=@n3.GrowthInput
  )
  n5 = ScalarParameter() {defaultValue=-0.6784, parameterName="DissolveTime0-1 (S)"}
  n6 = MaterialFunctionCall!(
   GrowthTime0-1=@n4
   GrowthBlendWidth_Inverse=ScalarParameter() {defaultValue=10, parameterName="GrowthBlendWidth_Inverse (S)"}
   GrowthBlendPower=ScalarParameter() {defaultValue=2, parameterName="GrowthBlendPower (S)"}
   ---------------
   TipAreaWidth_Inverse=ScalarParameter() {defaultValue=1.2, parameterName="TipAreaWidth_Inverse (S)"}
   TipAreaOffset=ScalarParameter() {defaultValue=-0.0768, parameterName="TipAreaOffset (S)"}
   TipAreaPower=ScalarParameter() {defaultValue=3, parameterName="TipAreaPower (S)"}
   ----------------
   DissolveTime0-1=LinearInterpolate(A=Add(A=@n4, B=@n5), B=@n3.DissolveInput, Alpha=@n3.DissolveInput)
   DissolveBlendWidth_Inverse=ScalarParameter() {defaultValue=1, parameterName="DissolveBlendWidth_Inverse (S)"}
   DissolveBlendPower=ScalarParameter() {defaultValue=2, parameterName="DissolveBlendPower (S)"}
   -----------------
   WPOUseBiNormal=Constant() {r=1}
   WPOStrength=@n3.Param3
  ) {materialFunction=/Game/Effects/Fort_Effects/GameplayPulginsAssets/PulseLancer/Slash/ProceduralSlashParameters}
  n7 = MaterialFunctionCall(Float2=TextureCoordinate()) {materialFunction=&BreakOutFloat2Components}
  n8 = VectorParameter() {
   group=TipTail
   parameterName=TipTail_EdgeMask_InvertControl
   channelNames {a="", b=Power, g=Multiply, r=Offset}
   defaultValue=<rgba 0 1.6418 2 1>
  }
  n9 = VectorParameter() {
   group=TipTail
   parameterName=TipTail_EdgeMask_Control
   channelNames {a="", b=Power, g=Multiply, r=Offset}
   defaultValue=<rgba 0 3.4075 1.2362 1>
  }
  n10 = Multiply(
   A=MaterialFunctionCall!(UorV=@n7.R, Invert_EdgeControl=@n8, NoInvert_EdgeControl=@n9) {materialFunction=/Game/Effects/Niagara/MasterMaterials/Functions/Common/FXCommon_EdgeMask_Ex}.Mask
   B=ScalarParameter() {defaultValue=8, parameterName=W}
  )
  Add(
   A=@n6.WPOWS
   B=MaterialFunctionCall(WidthBase=@n10, WidthTip=@n10, Expand U or V UV Channel=StaticBool()) {materialFunction=&SplineThicken}.WorldPositionOffset
  )
  n11 = VectorParameter() {parameterName=TipArea_Refine, defaultValue=<rgba 8 532.2567 20 0.5>}
  Add(A=MaterialFunctionCall(Texture=@n6.TipArea, RGB=@n11.RGB, A=@n11.RGB) {materialFunction=&MF_FX_Refine}.Result)
  n12 = VectorParameter() {parameterName=C, defaultValue=<rgba 1032.4696 895.6564 5000 1>}
  n13 = ParticleColor()
  MaterialFunctionCall(Emissive=Divide(A=Multiply(A=@n12.RGB, B=@n13.RGB)) {constB=1.97}) {materialFunction=&MF_FX_EmissiveAdjustments}
  n14 = VectorParameter() {
   group=TipTailFade
   parameterName=TipTailFade_EdgeMask_InvertControl
   channelNames {a="", b=Power, g=Multiply, r=Offset}
   defaultValue=<rgba 0 1 5 1>
  }
  n15 = VectorParameter() {
   group=TipTailFade
   parameterName=TipTailFade_EdgeMask_Control
   channelNames {a="", b=Power, g=Multiply, r=Offset}
   defaultValue=<rgba 0 3 4.1698 1>
  }
  n16 = VectorParameter() {group=Edge, parameterName=Edge_EdgeMask_Control, channelNames {a="", b=Power, g=Multiply, r=Offset}, defaultValue=<rgba 0 2.5 13.0231 1>}
  n17 = Multiply(
   A=Multiply(
    A=MaterialFunctionCall(UorV=@n7.R, Invert_EdgeControl=@n14, NoInvert_EdgeControl=@n15) {materialFunction=&FXCommon_EdgeMask_Ex}.Mask
    B=MaterialFunctionCall!(UorV=@n7.G, EdgeControl=@n16) {materialFunction=/Game/Effects/Niagara/MasterMaterials/Functions/Common/FXCommon_EdgeMask_Simple}.Mask
   )
   B=Multiply(A=@n6.Growth, B=@n6.Growth)
  )
  n18 = DepthFade(FadeDistance=ScalarParameter() {defaultValue=100, parameterName=DF})
  Multiply(A=Saturate(None=Multiply(A=@n17, B=@n13.A)), B=Multiply(A=Multiply(A=@n18, B=@n18), B=@n18))
  ScalarParameter() {parameterName="WPOStrength (S)"}
 ]
}
