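# Scratch debug material in the PulseLancer slash effects folder, left with its default name and Default Lit shading.
# A constant white color passes through MF_FX_EmissiveAdjustments, and the result feeds the engine's DebugFloat3Values function to display it as numeric digits.
# The debug readout connects to no material output, so the graph does not affect the rendered surface.
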Material /Game/Effects/Fort_Effects/GameplayPulginsAssets/PulseLancer/Slash/Material/NewMaterial
graph {
 properties {bAutomaticallySetUsageInEditor=false, usedShadingModels=Default Lit}
 nodes [MaterialFunctionCall!(
   Vector3=MaterialFunctionCall(Emissive=Constant3Vector() {constant=<rgba 1 1 1 0>}) {materialFunction=&MF_FX_EmissiveAdjustments}.Result
   MaximumNumberOfDigits
   UVs
   DebugTextLocation RG_UpperRight BA_LowerLeft
   Component Spacing
  ) {materialFunction=/Engine/Functions/Engine_MaterialFunctions02/Utility/DebugFloat3Values}]
}
